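; Gameplay ability for the durability version of the Witch Broom consumable that removes the broom once its ammo runs out.
; On activation it casts the current source weapon to the broom weapon class and stores it.
; If remaining ammo is zero or less, it unequips the weapon by inventory GUID and removes that item from the player's inventory through the controller.
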(event Ability|EventActivateAbility
  (bind _returnvalue (CallFunction|GetAvatarasFortPlayerPawn))
  (bind _witchbroomweap (Variables|Default|GetWitchBroomWeap))
  (bind _returnvalue_2 (FortWeapon|GetInventoryGUID _witchbroomweap))
  ;[1 1 1] Remove item from inventory
  (bind _asb_athena_witch_broom_weap (Utilities|Casting|CastToB_Athena_WitchBroom_Weap (Ability|GetCurrentSourceWeapon)))
  (Variables|Default|SetWitchBroomWeap _asb_athena_witch_broom_weap)
  (if (<= (FortWeapon|GetRemainingAmmo (Variables|Default|GetWitchBroomWeap)) 0)
    (Class|FortPlayerPawn|UnequipCurrentWeaponbyId _returnvalue _returnvalue_2 false)
    (Inventory|RemoveItemFromPlayerByGuid (Fortnite|GetFortPlayerControllerfromActor _returnvalue) _returnvalue_2 -1 true)))
